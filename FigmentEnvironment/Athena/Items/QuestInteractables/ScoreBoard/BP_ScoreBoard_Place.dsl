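; Scoreboard actor for a quest interactable minigame, showing current and high scores through dynamic sign materials.
; External events start games, add delayed increments and end games, where the server grants quest progress at the target score and updates the high score.
; Replicated score OnReps drive client material parameters, sounds and win effects.
; Outside Creative, clients bind quest updates after a hacky five-second delay and play success effects on the player upon completion.
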(event EventBeginPlay
  (bind _returnvalue (Fortnite|GetLocalFortPlayerControllers true))
  (bind _currentscorematindex (Variables|SignContent|GetCurrentScoreMatIndex))
  (bind _scoresignmesh (Variables|Default|GetScoreSignMesh))
  (bind _highscorematerial (Variables|SignContent|GetHighScoreMaterial))
  (bind _output (Utilities|Array|Get(acopy) _returnvalue))
  ;[0 0.9646 1] Initialize
  (Networking|SwitchHasAuthority
    (:Remote
      (Utilities|FlowControl|Sequence
        (:then_0
          ;; Base Score Setup_2
          (if (Utilities|NotEqual(Object) (Variables|SignContent|GetCurrentScoreMaterial) 0)
            (Rendering|Material|SetMaterial _scoresignmesh _currentscorematindex (Variables|SignContent|GetCurrentScoreMaterial)))
          (bind _returnvalue_2 (Rendering|Material|CreateDynamicMaterialInstance _scoresignmesh _currentscorematindex 0 "None" :!class /Script/Engine.PrimitiveComponent))
          (bind _output_get (Variables|ScoringInfo|SetMID_CurrentScore _returnvalue_2))
          (Rendering|Material|SetScalarParameterValue _output_get "Score" 0.0 :!class /Script/Engine.MaterialInstanceDynamic)
          (Rendering|Material|SetVectorParameterValue _output_get "Color" (Variables|SignContent|GetScoreColor) :!class /Script/Engine.MaterialInstanceDynamic))
        (:then_1
          ;; Setup High Score_2
          (if (Utilities|NotEqual(Object) _highscorematerial 0)
            (Rendering|Material|SetMaterial (Variables|Default|GetScoreSignMesh) (Variables|SignContent|GetHighScoreMatIndex) _highscorematerial))
          (bind _returnvalue_3 (Rendering|Material|CreateDynamicMaterialInstance (Variables|Default|GetScoreSignMesh) (Variables|SignContent|GetHighScoreMatIndex) (Variables|SignContent|GetHighScoreMaterial) "None" :!class /Script/Engine.PrimitiveComponent))
          (bind _output_get_2 (Variables|ScoringInfo|SetMID_HighScore _returnvalue_3))
          (Rendering|Material|SetScalarParameterValue _output_get_2 "Score" 0.0 :!class /Script/Engine.MaterialInstanceDynamic)
          (Rendering|Material|SetVectorParameterValue _output_get_2 "Color" (Variables|SignContent|GetHighScoreColor) :!class /Script/Engine.MaterialInstanceDynamic))
        (:then_2
          ;[1 1 1] Check if In Creative Mode or not
          (bind _asfort_game_state_athena (Utilities|Casting|CastToFortGameStateAthena (Game|GetGameState)))
          (if (Variables|/Script/FortniteGame.FortGameStateAthena|GetbPlaylistDataIsLoaded _asfort_game_state_athena)
            (if (GameplayTags|HasTag (Class|FortGameStateAthena|GetAthenaPlaylistContextTags _asfort_game_state_athena) "(TagName=\"Athena.Creative\")")
              (Variables|Default|SetCreativeMode true)
              (else
                (Variables|Default|SetCreativeMode false)))
            ;[1 0.5839 0] HackyHack Pls Fix Lietz
            (Utilities|FlowControl|Delay 5.0)
            (Networking|SwitchHasAuthority
              (:Remote
                ;[1 1 1] Quest Bindings
                (bind (_questisvalid _questcompleted _objectivecompleted) (Default|AthenaCheckQuestandObjectiveCompletion _output (Variables|Default|GetQuestItem) (Variables|Default|GetObjBackendName)))
                (if _questisvalid
                  (if _objectivecompleted
                    _
                    (elif (Variables|Default|GetCreativeMode)
                      _
                      (else
                        (Quests|BindEventtoOnQuestsUpdated (Quests|GetQuestManager _output "Invalid") (EventDispatchers|CreateEvent :!event Custom|QuestUpdated))
                        (Variables|Default|SetLocallyActive true)
                        (Development|PrintString "QUEST BOUND" false false "(R=0.000000,G=0.660000,B=1.000000,A=1.000000)" 1e+01)))))))
            (else
              (Playlist|BindEventtoOnCurrentPlaylistLoaded _asfort_game_state_athena (delegate Custom|PlaylistLoaded)))))))))

(event Custom|PlaylistLoaded
  (params (PlaylistName :type Name) (PlaylistContextTags :type Gameplay%20Tag%20Container%20Structure%20%28by%20ref%29))
  (bind _returnvalue (Fortnite|GetLocalFortPlayerControllers true))
  (bind _output (Utilities|Array|Get(acopy) _returnvalue))
  ;[1 1 1] Check if In Creative Mode or not
  (if (GameplayTags|HasTag PlaylistContextTags "(TagName=\"Athena.Creative\")")
    (Variables|Default|SetCreativeMode true)
    (else
      (Variables|Default|SetCreativeMode false)))
  ;[1 0.5839 0] HackyHack Pls Fix Lietz
  (Utilities|FlowControl|Delay 5.0)
  (Networking|SwitchHasAuthority
    (:Remote
      ;[1 1 1] Quest Bindings
      (bind (_questisvalid _questcompleted _objectivecompleted) (Default|AthenaCheckQuestandObjectiveCompletion _output (Variables|Default|GetQuestItem) (Variables|Default|GetObjBackendName)))
      (if _questisvalid
        (if _objectivecompleted
          _
          (elif (Variables|Default|GetCreativeMode)
            _
            (else
              (Quests|BindEventtoOnQuestsUpdated (Quests|GetQuestManager _output "Invalid") (EventDispatchers|CreateEvent :!event Custom|QuestUpdated))
              (Variables|Default|SetLocallyActive true)
              (Development|PrintString "QUEST BOUND" false false "(R=0.000000,G=0.660000,B=1.000000,A=1.000000)" 1e+01))))))))

(event Custom|QuestUpdated
  ;[1 1 1] Quest Updated
  (if (Variables|Default|GetLocallyActive)
    (bind (_questisvalid _questcompleted _objectivecompleted) (Default|AthenaCheckQuestandObjectiveCompletion (Utilities|Array|Get(acopy) (Fortnite|GetLocalFortPlayerControllers true)) (Variables|Default|GetQuestItem) (Variables|Default|GetObjBackendName)))
    (if (or _questcompleted _objectivecompleted)
      (Variables|Default|SetLocallyActive false)
      (Quests|UnbindEventfromOnQuestsUpdated (Quests|GetQuestManager (Utilities|Array|Get(acopy) (Fortnite|GetLocalFortPlayerControllers true)) "Invalid") (EventDispatchers|CreateEvent :!event Custom|QuestUpdated))
      (Utilities|FlowControl|Delay 1.5)
      (CallFunction|PlayQuestSuccessFX))))

(event Custom|PlayQuestSuccessFX
  (bind _returnvalue (Fortnite|GetLocalFortPlayerControllers true))
  (bind _output (Utilities|Array|Get(acopy) _returnvalue))
  (bind _returnvalue_2 (Game|Player|GetPlayerPawn _output))
  ;[1 1 1] Quest Success FX
  (Development|PrintString "SUCCESS FX" false false "(R=1.000000,G=0.703857,B=0.040648,A=1.000000)" 5.0)
  (Effects|SpawnFXSystemAttached (Variables|Default|GetQuestSuccessEmitter) (Variables|/Script/Engine.Character|GetMesh _returnvalue_2))
  (Audio|PlaySoundatLocation (Variables|Default|GetQuestSuccessSound) (Transformation|GetActorLocation _returnvalue_2)) ; Quest Success Sound
)

(event EventStartup
  (params (OwningPlayer :type Actor%20Object%20Reference))
  ;[1 1 1] Initiate
  ;[0 0.175 0.1025] Cast this to the right class
  (Variables|Default|SetLocalPlayer OwningPlayer)
  (Networking|SwitchHasAuthority
    (:Authority
      (Variables|ScoringInfo|SetCurrentScore 0))))

(event EventEndGame
  ;[0.3333 0.3333 0.3333] Win Condition and End Game
  ;;[0.2601 0.3249 1] Current Score exceeds High
  (Utilities|FlowControl|Sequence
    (:then_0
      ;[1 1 1] Check for Quest Update
      (if (>= (Variables|ScoringInfo|GetCurrentScore) (Variables|Default|GetQuestTargetGoal))
        (if (not (Variables|Default|GetCreativeMode))
          (Class|AthenaQuestFunctionLibrary|AthenaGiveQuestUpdatetoPlayers (Variables|Default|GetQuestItem) (Variables|Default|GetObjBackendName) (Variables|Default|GetObjStat) (Utilities|Array|MakeArray (Fortnite|GetFortPlayerControllerfromActor (Variables|Default|GetLocalPlayer)))))))
    (:then_1
      (if (> (Variables|ScoringInfo|GetCurrentScore) (Variables|ScoringInfo|GetHighScore))
        (Development|PrintString "End Game HIGH SCORE" false false "(R=0.224283,G=1.000000,B=0.000000,A=1.000000)" 5.0)
        ;;[0.3906 0.5729 0.1969] Set Final Score
        (Variables|ScoringInfo|SetHighScore (Variables|ScoringInfo|GetCurrentScore))
        (else
          (Development|PrintString "End Game NO HIGH SCORE" false false "(R=1.000000,G=0.035352,B=0.000000,A=1.000000)" 5.0)
          (CallFunction|NoHighScoreFX)))
      (Variables|ScoringInfo|SetCurrentScore 0))))

(event Custom|WinFX
  ;[1 0.9976 0] Win Behavior on board / Client Only
  (Networking|SwitchHasAuthority
    (:Remote
      (Audio|PlaySoundatLocation (Variables|SignContent|GetNewHighScoreSound) (Transformation|GetActorLocation)) ; High Score Sound
      (Utilities|FlowControl|Delay (Variables|ScoringInfo|GetHighScoreDelay))
      (Components|Activation|Activate (Variables|Default|GetHighScoreFX) false)
      (Audio|PlaySoundatLocation (Variables|SignContent|GetFireworksSound) (Transformation|GetActorLocation)) ; High Score Sound
      (Utilities|FlowControl|Delay 1.5)
      (Rendering|Material|SetScalarParameterValue (Variables|ScoringInfo|GetMID_HighScore) "Win" 0.0 :!class /Script/Engine.MaterialInstanceDynamic))))

(event Custom|No%20High%20Score%20FX
  ;[1 0.08653 0.7576] AUDIO - Play lost game sound
  (Networking|SwitchHasAuthority
    (:Remote
      (Audio|PlaySoundatLocation (Variables|SignContent|GetNoHighScoreSound) (Transformation|GetActorLocation) "0, 0, 0" 2.0) ; Amplified to be audible
    )))

(event EventMoveScore
  (params (ScoreIncrement :type Integer))
  ;[0.2917 0.2917 0.2917] Score Behavior
  ;;[0.3217 0.8799 1] External Event to move Score up (or down)
  (Utilities|FlowControl|Delay (Variables|ScoringInfo|GetScoreDelay))
  (Variables|ScoringInfo|SetCurrentScore (+ (Variables|ScoringInfo|GetCurrentScore) ScoreIncrement)))

(fn OnRep_CurrentScore
  (bind _currentscore (Variables|ScoringInfo|GetCurrentScore))
  (Networking|SwitchHasAuthority
    (:Remote
      (if (not (== _currentscore 0))
        (Audio|PlaySoundatLocation (Variables|SignContent|GetScore%20Up%20Sound) (Transformation|GetWorldLocation (Variables|Default|GetScoreSignMesh))) ; Score Up Sound
      )
      (Rendering|Material|SetScalarParameterValue (Variables|ScoringInfo|GetMID_CurrentScore) "Score" (Math|Conversions|ToFloat(Integer) _currentscore) :!class /Script/Engine.MaterialInstanceDynamic))))

(fn OnRep_CurrentScoreMatIndex)

(fn OnRep_HighScore
  ;[0.4972 1 0.5806] Write scores to board
  (Networking|SwitchHasAuthority
    (:Remote
      (Rendering|Material|SetScalarParameterValue (Variables|ScoringInfo|GetMID_CurrentScore) "Score" 0.0 :!class /Script/Engine.MaterialInstanceDynamic)
      ;[1 0.3126 0] Play final effects
      (CallFunction|WinFX)
      (Rendering|Material|SetScalarParameterValue (Variables|ScoringInfo|GetMID_HighScore) "Score" (Math|Conversions|ToFloat(Integer) (Variables|ScoringInfo|GetHighScore)) :!class /Script/Engine.MaterialInstanceDynamic))))

(fn OnRep_HighScoreColor)

(fn ConstructionScript
  (bind _sign_mesh (Variables|SignContent|GetSign%20Mesh))
  ; Mesh Setup
  (if (Utilities|NotEqual(Object) _sign_mesh 0)
    (Components|StaticMesh|SetStaticMesh (Variables|Default|GetScoreSignMesh) _sign_mesh))
  ; Win Particle Setup
  (Effects|Components|ParticleSystem|SetTemplate (Variables|Default|GetHighScoreFX) (Variables|SignContent|GetWinEffect)))
