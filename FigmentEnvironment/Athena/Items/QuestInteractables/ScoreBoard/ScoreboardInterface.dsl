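; Blueprint interface for a quest scoreboard interactable in the Figment environment.
; It declares three functions: Startup takes the owning player actor, MoveScore takes an integer score increment, and EndGame takes no parameters.
; The file holds only these signatures, leaving the behavior to implementing Blueprints.
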(fn EndGame
  (return))

(fn MoveScore
  (params (ScoreIncrement :type Integer))
  (return))

(fn Startup
  (params (OwningPlayer :type Actor%20Object%20Reference)))
